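# Material function that builds the pre-FX material for the BlazePuff umbrella glider skin.
# MF_UmbrellaPuff_BaseInputs supplies face and surface layers and MF_UmbrellaPuff_Jelly the sprite layer, which are blended by masks, with tunable emission mixing on the handle, then passed through MF_Sprite_DitherFresnel.
# A vertex-interpolated handle mask comes from radius and height thresholds on pre-skinned local position.
# It also outputs that position, valid in both shader stages, and a debug vector.
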MaterialFunction /BRCosmetics/Gadgets/Assets/VinderTech_GliderChute/Glider_Solo_MarkII/Skins/BlazePuff/Materials/MaterialFunctions/MF_UmbrellaPuff_PreFX
graph {nodes [
  n1 = ScalarParameter() {defaultValue=1, group=Umbrella | Handle, parameterName="Sprite Emission -> Handle Diffuse Mix", sliderMax=1, sortPriority=40}
  n2 = GetMaterialAttributes(None=NamedRerouteUsage()) {attributeGetTypes [#1]}
  n3 = ScalarParameter() {group=Umbrella | Handle, parameterName="Sprite Emission -> Handle Emission Mix", sliderMax=1, sortPriority=40}
  n4 = NamedRerouteUsage()
  n5 = ScalarParameter() {defaultValue=1, group=Umbrella | Handle, parameterName=Surface Mask Mult, sliderMax=1, sortPriority=0}
  n6 = BlendMaterialAttributes(
   A=NamedRerouteUsage()
   B=SetMaterialAttributes!(MaterialAttributes=NamedRerouteUsage(), BaseColor=Multiply(A=@n1, B=@n2.EmissiveColor), EmissiveColor=Multiply(A=@n2.EmissiveColor, B=@n3)) {attributeSetTypes [#2, #1]}
   Alpha=Multiply(A=Multiply(A=@n4, B=@n5), B=NamedRerouteUsage())
  ) {vertexAttributeBlendType=UseA}
  n7 = NamedRerouteUsage()
  n8 = NamedRerouteUsage()
  FunctionOutput(None=MaterialFunctionCall!(MA=BlendMaterialAttributes(A=@n6, B=@n7, Alpha=@n8) {vertexAttributeBlendType=UseA}) {materialFunction=/BRCosmetics/Animation/Props/Gadgets_PixieParts/Materials/ParentMaterials/MF_Sprite_DitherFresnel}.MA)
  n9 = LocalPosition() {localOrigin=InstancePreSkinning}
  n10 = VertexInterpolator(VS=@n9.XYZ)
  FunctionOutput(None=NamedRerouteDeclaration(None=ShaderStageSwitch(PixelShader=@n10.PS, VertexShader=@n9.XYZ)) {name=PreSkinnedPos_VS-PS, nodeColor=<rgba 0 0.4844 1 1>}) {outputName=PreskinnedLocalPos, sortPriority=10}
  FunctionOutput(None=NamedRerouteUsage()) {outputName="DEBUG (V3)", sortPriority=100}
  n11 = NamedRerouteUsage()
  n12 = MaterialFunctionCall!(
   Surface Normal WS=NamedRerouteDeclaration(None=Transform(None=@n11)) {name="Surface Normal WS (V3)", nodeColor=<rgba 0.9922 0 1 1>}
   Face UVs=NamedRerouteUsage()
   Handle Mask=NamedRerouteUsage()
   FXMask
  ) {materialFunction=&/MF_UmbrellaPuff_Jelly}
  NamedRerouteDeclaration(None=@n12.Mask_InnerRadialGlow) {name=Mask_InnerRadialGlow, nodeColor=<rgba 0 1 0.5781 1>}
  NamedRerouteDeclaration(None=@n12.Mask_Logo) {name=Mask_Logo, nodeColor=<rgba 0 1 0.5781 1>}
  n13 = MaterialFunctionCall() {materialFunction=&/MF_UmbrellaPuff_BaseInputs}
  NamedRerouteDeclaration(None=@n13."Face Material (MA)") {name="Face Material (MA)", nodeColor=<rgba 0 1 0.1797 1>}
  NamedRerouteDeclaration(None=@n13."Face Mask (S)") {name="Face Mask (S)", nodeColor=<rgba 1 0 0.1406 1>}
  NamedRerouteDeclaration(None=@n13."Face UVs (V2)") {name="Face UVs (V2)", nodeColor=<rgba 0 1 0.7422 1>}
  MaterialExpressionLocalPosition_0 = LocalPosition() {includedOffsets=ExcludeOffsets, localOrigin=InstancePreSkinning}
  n14 = Convert(None=@MaterialExpressionLocalPosition_0.XYZ) {
   convertInputs [{
     type=Vector3
     expressionInput {
      expression=&self:MaterialExpressionLocalPosition_0
      inputName=None
      mask=1
      maskA=0
      maskB=1
      maskG=1
      maskR=1
      outputIndex=0
     }
    }]
   convertOutputs [{type=Vector2}, {type=Scalar}]
  }
  n15 = Length(None=@n14) {is=class MaterialExpressionLength}
  n16 = OneMinus(None=Step(X=Divide(A=@n15) {constB=12}) {constY=1})
  n17 = OneMinus(None=Step(X=@n14) {constY=109.4})
  NamedRerouteDeclaration(None=VertexInterpolator(VS=Multiply(A=@n16, B=@n17)).PS) {name="Handle Mask (S)", nodeColor=<rgba 0.5938 0 1 1>}
  NamedRerouteDeclaration(None=@n12.Mask_SpriteFresnel) {name=Mask_SpriteFresnel, nodeColor=<rgba 0 1 0.5781 1>}
  NamedRerouteDeclaration(None=@n12.Sprite WPO) {name=Sprite WPO, nodeColor=<rgba 0 1 0.5781 1>}
  NamedRerouteDeclaration(None=@n13."DEBUG (V3)") {name="DEBUG (V3)", nodeColor=<rgba 1 0 0 1>}
  NamedRerouteDeclaration(None=@n13."Surface Mask (S)") {name="Surface Mask (S)", nodeColor=<rgba 0.6172 1 0 1>}
  n18 = GetMaterialAttributes(None=NamedRerouteUsage()) {attributeGetTypes [#3, #4]}
  n19 = ScalarParameter() {group=Umbrella | Handle, parameterName=Surface Specular Mult, sortPriority=10}
  n20 = Multiply(
   A=Power(Base=@n18.Roughness, Exp=ScalarParameter() {defaultValue=1, group=Umbrella | Handle, parameterName=Surface Roughness Exp, sortPriority=20})
   B=ScalarParameter() {defaultValue=1, group=Umbrella | Handle, parameterName=Surface Roughness Mult, sortPriority=30}
  )
  NamedRerouteDeclaration(None=SetMaterialAttributes!(MaterialAttributes=@n18.MaterialAttributes, Specular=Multiply(A=@n18.Specular, B=@n19), Roughness=Saturate(None=@n20), Metallic=Constant()) {attributeSetTypes [#3, #4, #5]}) {name="Surface Material (MF)", nodeColor=<rgba 0 0.6016 1 1>}
  NamedRerouteDeclaration(None=@n12.MA) {name="Sprite Material (MA)", nodeColor=<rgba 0.9219 1 0 1>}
  NamedRerouteDeclaration(None=@n13."Surface Normal TS (V3)") {name="Surface Normal TS (V3)", nodeColor=<rgba 1 0 0.9375 1>}
  NamedRerouteDeclaration(None=@n13."Surface Material (MA)") {name="Surface Material In (MF)", nodeColor=<rgba 0 0.6016 1 1>}
 ]}
