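# Parent material for the BlazePuff skin of the Solo MarkII glider umbrella, a Default Lit skeletal-mesh material built on material attributes.
# A dedicated UmbrellaPuff PreFX function builds the surface, which passes through the shared character brightness and scale balance function.
# The shared 2023 character PostFX function then takes that result, along with pre-skinned local position, a Fresnel term and a constant mask vector.
# A named reroute exposes the PreFX debug vector.
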Material /BRCosmetics/Gadgets/Assets/VinderTech_GliderChute/Glider_Solo_MarkII/Skins/BlazePuff/Materials/M_UmbrellaPuff_Parent
graph {
 properties {bAutomaticallySetUsageInEditor=false, bUseMaterialAttributes=true, bUsedWithSkeletalMesh=true, usedShadingModels=Default Lit}
 nodes [
  n1 = MaterialFunctionCall() {materialFunction=/BRCosmetics/Gadgets/Assets/VinderTech_GliderChute/Glider_Solo_MarkII/Skins/BlazePuff/Materials/MaterialFunctions/MF_UmbrellaPuff_PreFX}
  n2 = MaterialFunctionCall!(Material=@n1.Result) {materialFunction=/Game/Packages/Fortress_SharedMaterials/Base_Material_Functions/CharacterBrightnessandScale_Balance}
  MaterialFunctionCall(
   Material=@n2.Result
   M=Constant3Vector() {constant=<rgba 1 1 0 0>}
   Pre Skinned Local Position=@n1.PreskinnedLocalPos
   Distance Blend=@n2.Result
   Fresnel=Fresnel() {baseReflectFraction=0, exponent=1}
  ) {materialFunction=&MF_CharacterParent_2023_PostFX}
  NamedRerouteDeclaration(None=@n1."DEBUG (V3)") {name="DEBUG (V3)", nodeColor=<rgba 1 0 0 1>}
 ]
}
